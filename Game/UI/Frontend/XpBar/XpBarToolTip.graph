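# Tooltip widget for the frontend XP bar, showing a breakdown of account, boosted and daily bonus XP.
# A shadowed, size-constrained border stacks a header, color-swatch rows for each XP type and a body line, separated by thin dividers.
# Five getter functions each return a matching variable for header, body, commander, boost and rested text.
# The event graph is empty, and nothing in this widget sets those variables.
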WidgetBlueprint /Game/UI/Frontend/XpBar/XpBarToolTip
graphs [
 EventGraph {links [], nodes []}
 GetBodyText {nodes [
   GetBodyText() {is=title K2Node_FunctionEntry}
   > Return Node!(execute, ReturnValue:Text=Get BodyText() -> (BodyText:Text))
  ]}
 GetBoostXPText {nodes [
   GetBoostXPText() {is=title K2Node_FunctionEntry}
   > Return Node(ReturnValue=Get BoostXPText() -> (BoostXPText:Text))
  ]}
 GetCommanderXPText {nodes [
   GetCommanderXPText() {is=title K2Node_FunctionEntry}
   > Return Node(ReturnValue=Get CommanderXPText() -> (CommanderXPText:Text))
  ]}
 GetHeaderText {nodes [
   GetHeaderText() {is=title K2Node_FunctionEntry}
   > Return Node(ReturnValue=Get HeaderText() -> (HeaderText:Text))
  ]}
 GetRestedXPText {nodes [
   GetRestedXPText() {is=title K2Node_FunctionEntry}
   > Return Node(ReturnValue=Get RestedXPText() -> (RestedXPText:Text))
  ]}
]
variables [
 BodyText {type=string, default=""}
 BoostXPText {type=string, default=""}
 CommanderXPText {type=string, default=""}
 HeaderText {type=string, default=""}
 RestedXPText {type=string, default=""}
]
widgets {
 info {inheritedWidgetCount=0, parentClass=&CommonUserWidget, rootWidgetClass=&SizeBox, widgetCount=33}
 widgets [
  SizeBox_0 {
   widgetClassPath=&SizeBox
   widget=SizeBox_0
   properties {
    bOverride_MaxDesiredWidth=true
    bOverride_MinDesiredHeight=true
    bOverride_MinDesiredWidth=true
    maxDesiredWidth=512
    minDesiredHeight=32
    minDesiredWidth=384
   }
  }
  Overlay_0 {widgetClassPath=&Overlay, parent=SizeBox_0, widget=Overlay_0, slot {class=SizeBoxSlot}}
  CommonBorder_1 {
   widgetClassPath=&CommonBorder
   parent=Overlay_0
   widget=CommonBorder_1
   properties {
    style=/FortUILibrary/Styles/Border/Border-SolidBG-ToolTipShadow.Border-SolidBG-ToolTipShadow_C
    renderTransform {angle=0, scale=<1 1>, shear=<0 0>, translation=<6 8>}
   }
   slot {class=OverlaySlot, properties {horizontalAlignment=HAlign_Fill, verticalAlignment=VAlign_Fill}}
  }
  CommonBorder_0 {
   widgetClassPath=&CommonBorder
   parent=Overlay_0
   widget=CommonBorder_0
   properties {style=/FortUILibrary/Styles/Border/Border-SolidBG-ToolTip.Border-SolidBG-ToolTip_C, padding=<ltrb 12 12 12 12>}
   slot {class=OverlaySlot, properties {horizontalAlignment=HAlign_Fill, verticalAlignment=VAlign_Fill}}
  }
  VerticalBox_20 {
   widgetClassPath=&VerticalBox
   parent=CommonBorder_0
   widget=VerticalBox_20
   slot {class=BorderSlot, properties {padding=<ltrb 12 12 12 12>}}
  }
  HorizontalBox_0 {
   widgetClassPath=&HorizontalBox
   parent=VerticalBox_20
   widget=HorizontalBox_0
   slot {class=VerticalBoxSlot, properties {padding=<ltrb 0 0 0 4>}}
  }
  CommonTextBlock_0 {
   widgetClassPath=&CommonTextBlock
   parent=HorizontalBox_0
   widget=CommonTextBlock_0
   properties {
    justification=Center
    style=/FortUILibrary/Styles/Typography/Foundation/TextStyle-Base-ToolTip-B.TextStyle-Base-ToolTip-B_C
    text=Header Text
    font {
     bForceMonospaced=false
     bMaterialIsStencil=false
     fontMaterial=null
     fontObject=&NotoSans
     letterSpacing=0
     monospacedWidth=1
     size=12
     skewAmount=0
     typefaceFontName=NotoSans-Bold
     outlineSettings {
      bApplyOutlineToDropShadows=false
      bMiteredCorners=false
      bSeparateFillAlpha=false
      outlineMaterial=null
      outlineSize=0
      outlineColor=<rgba 0 0 0 1>
     }
    }
    shadowOffset=<0 0>
   }
   slot {class=HorizontalBoxSlot, properties {horizontalAlignment=HAlign_Left, verticalAlignment=VAlign_Center, size {sizeRule=Fill, value=1}}}
  }
  VerticalBox_0 {
   widgetClassPath=&VerticalBox
   parent=VerticalBox_20
   widget=VerticalBox_0
   slot {class=VerticalBoxSlot, properties {size {sizeRule=Fill, value=1}}}
  }
  Image_3 {
   widgetClassPath=&Image
   bIsVariable=true
   parent=VerticalBox_0
   widget=Image_3
   properties {
    brush {
     bIsDynamicallyLoaded=false
     drawAs=Image
     imageType=NoImage
     mirroring=NoMirror
     resourceName=None
     resourceObject=null
     tiling=Horizontal
     imageSize=<6 2>
     margin=<ltrb 0 0 0 0>
     outlineSettings {
      bUseBrushTransparency=false
      roundingType=HalfHeightRadius
      width=0
      color {colorUseRule=UseColor_Specified, specifiedColor=<rgba 0 0 0 0>}
      cornerRadii=<0 0 0 0>
     }
     tintColor {colorUseRule=UseColor_Specified, specifiedColor=<rgba 1 1 1 1>}
     uVRegion {bIsValid=false, max=<0 0>, min=<0 0>}
    }
    colorAndOpacity=<rgba 1 1 1 0.2>
   }
   slot {class=VerticalBoxSlot, properties {padding=<ltrb 0 0 0 4>, size {sizeRule=Fill, value=1}}}
  }
  VerticalBox_27 {widgetClassPath=&VerticalBox, parent=VerticalBox_20, widget=VerticalBox_27, slot {class=VerticalBoxSlot}}
  HorizontalBox_15 {widgetClassPath=&HorizontalBox, parent=VerticalBox_27, widget=HorizontalBox_15, slot {class=VerticalBoxSlot}}
  Image_0 {
   widgetClassPath=&Image
   bIsVariable=true
   parent=HorizontalBox_15
   widget=Image_0
   properties {brush {
     bIsDynamicallyLoaded=false
     drawAs=Image
     imageType=NoImage
     mirroring=NoMirror
     resourceName=None
     resourceObject=null
     tiling=NoTile
     imageSize=<32 12>
     margin=<ltrb 0 0 0 0>
     outlineSettings {
      bUseBrushTransparency=false
      roundingType=HalfHeightRadius
      width=0
      color {colorUseRule=UseColor_Specified, specifiedColor=<rgba 0 0 0 0>}
      cornerRadii=<0 0 0 0>
     }
     tintColor {colorUseRule=UseColor_Specified, specifiedColor=<rgba 1 0.6939 0 1>}
     uVRegion {bIsValid=false, max=<0 0>, min=<0 0>}
    }}
   slot {class=HorizontalBoxSlot, properties {verticalAlignment=VAlign_Center, padding=<ltrb 0 0 4 0>}}
  }
  HorizontalBox_18 {
   widgetClassPath=&HorizontalBox
   parent=HorizontalBox_15
   widget=HorizontalBox_18
   slot {class=HorizontalBoxSlot, properties {horizontalAlignment=HAlign_Left, padding=<ltrb 4 4 4 4>, size {sizeRule=Fill, value=1}}}
  }
  CommonTextBlock_25 {
   widgetClassPath=&CommonTextBlock
   parent=HorizontalBox_18
   widget=CommonTextBlock_25
   properties {
    justification=Center
    style=/FortUILibrary/Styles/Typography/Foundation/TextStyle-Base-ToolTip-Body.TextStyle-Base-ToolTip-Body_C
    text=Account XP :
    colorAndOpacity {colorUseRule=UseColor_Specified, specifiedColor=<rgba 0.4452 0.5972 1 1>}
    font {
     bForceMonospaced=false
     bMaterialIsStencil=false
     fontMaterial=null
     fontObject=&NotoSans
     letterSpacing=0
     monospacedWidth=1
     size=12
     skewAmount=0
     typefaceFontName=NotoSans-Regular
     outlineSettings {
      bApplyOutlineToDropShadows=false
      bMiteredCorners=false
      bSeparateFillAlpha=false
      outlineMaterial=null
      outlineSize=0
      outlineColor=<rgba 0 0 0 1>
     }
    }
    shadowOffset=<0 0>
   }
   slot {class=HorizontalBoxSlot, properties {verticalAlignment=VAlign_Center}}
  }
  Spacer_0 {widgetClassPath=&Spacer, parent=HorizontalBox_18, widget=Spacer_0, properties {size=<4 1>}, slot {class=HorizontalBoxSlot}}
  CommonTextBlock_29 {
   widgetClassPath=&CommonTextBlock
   parent=HorizontalBox_18
   widget=CommonTextBlock_29
   properties {
    style=&TextStyle-Base-ToolTip-Body_C
    text=Command XP Text
    colorAndOpacity {colorUseRule=UseColor_Specified, specifiedColor=<rgba 0.4452 0.5972 1 1>}
    font {
     bForceMonospaced=false
     bMaterialIsStencil=false
     fontMaterial=null
     fontObject=&NotoSans
     letterSpacing=0
     monospacedWidth=1
     size=12
     skewAmount=0
     typefaceFontName=NotoSans-Regular
     outlineSettings {
      bApplyOutlineToDropShadows=false
      bMiteredCorners=false
      bSeparateFillAlpha=false
      outlineMaterial=null
      outlineSize=0
      outlineColor=<rgba 0 0 0 1>
     }
    }
    shadowOffset=<0 0>
   }
   slot {class=HorizontalBoxSlot, properties {verticalAlignment=VAlign_Center}}
  }
  HorizontalBox_16 {widgetClassPath=&HorizontalBox, parent=VerticalBox_27, widget=HorizontalBox_16, slot {class=VerticalBoxSlot}}
  Image_1 {
   widgetClassPath=&Image
   bIsVariable=true
   parent=HorizontalBox_16
   widget=Image_1
   properties {brush {
     bIsDynamicallyLoaded=false
     drawAs=Image
     imageType=NoImage
     mirroring=NoMirror
     resourceName=None
     resourceObject=null
     tiling=NoTile
     imageSize=<32 12>
     margin=<ltrb 0 0 0 0>
     outlineSettings {
      bUseBrushTransparency=false
      roundingType=HalfHeightRadius
      width=0
      color {colorUseRule=UseColor_Specified, specifiedColor=<rgba 0 0 0 0>}
      cornerRadii=<0 0 0 0>
     }
     tintColor {colorUseRule=UseColor_Specified, specifiedColor=<rgba 0.0343 0.1022 0.2502 1>}
     uVRegion {bIsValid=false, max=<0 0>, min=<0 0>}
    }}
   slot {class=HorizontalBoxSlot, properties {verticalAlignment=VAlign_Center, padding=<ltrb 0 0 4 0>}}
  }
  HorizontalBox_20 {
   widgetClassPath=&HorizontalBox
   parent=HorizontalBox_16
   widget=HorizontalBox_20
   slot {class=HorizontalBoxSlot, properties {horizontalAlignment=HAlign_Left, padding=<ltrb 4 4 4 4>, size {sizeRule=Fill, value=1}}}
  }
  CommonTextBlock_26 {
   widgetClassPath=&CommonTextBlock
   parent=HorizontalBox_20
   widget=CommonTextBlock_26
   properties {
    style=&TextStyle-Base-ToolTip-Body_C
    text=Boosted XP :
    colorAndOpacity {colorUseRule=UseColor_Specified, specifiedColor=<rgba 0.4452 0.5972 1 1>}
    font {
     bForceMonospaced=false
     bMaterialIsStencil=false
     fontMaterial=null
     fontObject=&NotoSans
     letterSpacing=0
     monospacedWidth=1
     size=12
     skewAmount=0
     typefaceFontName=NotoSans-Regular
     outlineSettings {
      bApplyOutlineToDropShadows=false
      bMiteredCorners=false
      bSeparateFillAlpha=false
      outlineMaterial=null
      outlineSize=0
      outlineColor=<rgba 0 0 0 1>
     }
    }
    shadowOffset=<0 0>
   }
   slot {class=HorizontalBoxSlot, properties {verticalAlignment=VAlign_Center}}
  }
  Spacer_1 {widgetClassPath=&Spacer, parent=HorizontalBox_20, widget=Spacer_1, properties {size=<4 1>}, slot {class=HorizontalBoxSlot}}
  CommonTextBlock_30 {
   widgetClassPath=&CommonTextBlock
   parent=HorizontalBox_20
   widget=CommonTextBlock_30
   properties {
    justification=Center
    style=&TextStyle-Base-ToolTip-Body_C
    text=Boosted XP Text
    colorAndOpacity {colorUseRule=UseColor_Specified, specifiedColor=<rgba 0.4452 0.5972 1 1>}
    font {
     bForceMonospaced=false
     bMaterialIsStencil=false
     fontMaterial=null
     fontObject=&NotoSans
     letterSpacing=0
     monospacedWidth=1
     size=12
     skewAmount=0
     typefaceFontName=NotoSans-Regular
     outlineSettings {
      bApplyOutlineToDropShadows=false
      bMiteredCorners=false
      bSeparateFillAlpha=false
      outlineMaterial=null
      outlineSize=0
      outlineColor=<rgba 0 0 0 1>
     }
    }
    shadowOffset=<0 0>
   }
   slot {class=HorizontalBoxSlot, properties {verticalAlignment=VAlign_Center}}
  }
  HorizontalBox_17 {
   widgetClassPath=&HorizontalBox
   parent=VerticalBox_27
   widget=HorizontalBox_17
   slot {class=VerticalBoxSlot, properties {padding=<ltrb 0 0 0 4>}}
  }
  Image_2 {
   widgetClassPath=&Image
   bIsVariable=true
   parent=HorizontalBox_17
   widget=Image_2
   properties {brush {
     bIsDynamicallyLoaded=false
     drawAs=Image
     imageType=NoImage
     mirroring=NoMirror
     resourceName=None
     resourceObject=null
     tiling=NoTile
     imageSize=<32 12>
     margin=<ltrb 0 0 0 0>
     outlineSettings {
      bUseBrushTransparency=false
      roundingType=HalfHeightRadius
      width=0
      color {colorUseRule=UseColor_Specified, specifiedColor=<rgba 0 0 0 0>}
      cornerRadii=<0 0 0 0>
     }
     tintColor {colorUseRule=UseColor_Specified, specifiedColor=<rgba 0.0561 0.2423 0.1221 1>}
     uVRegion {bIsValid=false, max=<0 0>, min=<0 0>}
    }}
   slot {class=HorizontalBoxSlot, properties {verticalAlignment=VAlign_Center, padding=<ltrb 0 0 4 0>}}
  }
  HorizontalBox_22 {
   widgetClassPath=&HorizontalBox
   parent=HorizontalBox_17
   widget=HorizontalBox_22
   slot {class=HorizontalBoxSlot, properties {horizontalAlignment=HAlign_Left, padding=<ltrb 4 4 4 4>, size {sizeRule=Fill, value=1}}}
  }
  CommonTextBlock_28 {
   widgetClassPath=&CommonTextBlock
   parent=HorizontalBox_22
   widget=CommonTextBlock_28
   properties {
    style=&TextStyle-Base-ToolTip-Body_C
    text=Daily Bonus XP :
    colorAndOpacity {colorUseRule=UseColor_Specified, specifiedColor=<rgba 0.4452 0.5972 1 1>}
    font {
     bForceMonospaced=false
     bMaterialIsStencil=false
     fontMaterial=null
     fontObject=&NotoSans
     letterSpacing=0
     monospacedWidth=1
     size=12
     skewAmount=0
     typefaceFontName=NotoSans-Regular
     outlineSettings {
      bApplyOutlineToDropShadows=false
      bMiteredCorners=false
      bSeparateFillAlpha=false
      outlineMaterial=null
      outlineSize=0
      outlineColor=<rgba 0 0 0 1>
     }
    }
    shadowOffset=<0 0>
   }
   slot {class=HorizontalBoxSlot, properties {verticalAlignment=VAlign_Center}}
  }
  Spacer_2 {widgetClassPath=&Spacer, parent=HorizontalBox_22, widget=Spacer_2, properties {size=<4 1>}, slot {class=HorizontalBoxSlot}}
  CommonTextBlock_31 {
   widgetClassPath=&CommonTextBlock
   parent=HorizontalBox_22
   widget=CommonTextBlock_31
   properties {
    justification=Center
    style=&TextStyle-Base-ToolTip-Body_C
    text=Rested XP Text
    colorAndOpacity {colorUseRule=UseColor_Specified, specifiedColor=<rgba 0.4452 0.5972 1 1>}
    font {
     bForceMonospaced=false
     bMaterialIsStencil=false
     fontMaterial=null
     fontObject=&NotoSans
     letterSpacing=0
     monospacedWidth=1
     size=12
     skewAmount=0
     typefaceFontName=NotoSans-Regular
     outlineSettings {
      bApplyOutlineToDropShadows=false
      bMiteredCorners=false
      bSeparateFillAlpha=false
      outlineMaterial=null
      outlineSize=0
      outlineColor=<rgba 0 0 0 1>
     }
    }
    shadowOffset=<0 0>
   }
   slot {class=HorizontalBoxSlot, properties {verticalAlignment=VAlign_Center}}
  }
  VerticalBox_2 {
   widgetClassPath=&VerticalBox
   parent=VerticalBox_20
   widget=VerticalBox_2
   slot {class=VerticalBoxSlot, properties {size {sizeRule=Fill, value=1}}}
  }
  Image_5 {
   widgetClassPath=&Image
   bIsVariable=true
   parent=VerticalBox_2
   widget=Image_5
   properties {brush {
     bIsDynamicallyLoaded=false
     drawAs=Image
     imageType=NoImage
     mirroring=NoMirror
     resourceName=None
     resourceObject=null
     tiling=Horizontal
     imageSize=<6 2>
     margin=<ltrb 0 0 0 0>
     outlineSettings {
      bUseBrushTransparency=false
      roundingType=HalfHeightRadius
      width=0
      color {colorUseRule=UseColor_Specified, specifiedColor=<rgba 0 0 0 0>}
      cornerRadii=<0 0 0 0>
     }
     tintColor {colorUseRule=UseColor_Specified, specifiedColor=<rgba 1 1 1 0.2>}
     uVRegion {bIsValid=false, max=<0 0>, min=<0 0>}
    }}
   slot {class=VerticalBoxSlot, properties {padding=<ltrb 0 0 0 4>, size {sizeRule=Fill, value=1}}}
  }
  HorizontalBox_6 {widgetClassPath=&HorizontalBox, parent=VerticalBox_20, widget=HorizontalBox_6, slot {class=VerticalBoxSlot}}
  Spacer_3 {widgetClassPath=&Spacer, parent=HorizontalBox_6, widget=Spacer_3, properties {size=<40 1>}, slot {class=HorizontalBoxSlot}}
  CommonTextBlock_2 {
   widgetClassPath=&CommonTextBlock
   parent=HorizontalBox_6
   widget=CommonTextBlock_2
   properties {
    justification=Center
    style=&TextStyle-Base-ToolTip-Body_C
    text=XP Text
    colorAndOpacity {colorUseRule=UseColor_Specified, specifiedColor=<rgba 0.4452 0.5972 1 1>}
    font {
     bForceMonospaced=false
     bMaterialIsStencil=false
     fontMaterial=null
     fontObject=&NotoSans
     letterSpacing=0
     monospacedWidth=1
     size=12
     skewAmount=0
     typefaceFontName=NotoSans-Regular
     outlineSettings {
      bApplyOutlineToDropShadows=false
      bMiteredCorners=false
      bSeparateFillAlpha=false
      outlineMaterial=null
      outlineSize=0
      outlineColor=<rgba 0 0 0 1>
     }
    }
    shadowOffset=<0 0>
   }
   slot {class=HorizontalBoxSlot, properties {horizontalAlignment=HAlign_Left, size {sizeRule=Fill, value=1}}}
  }
 ]
}
